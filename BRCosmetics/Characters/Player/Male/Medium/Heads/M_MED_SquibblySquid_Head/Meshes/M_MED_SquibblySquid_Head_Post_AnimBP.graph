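# Post-process anim blueprint for the SquibblySquid head cosmetic that makes its tentacle bones wiggle and bounce.
# Each update, the event graph builds three rotators from game-time sine waves at differing speeds and amplitudes.
# The anim graph adds these to three dyn_hair bone pairs in bone space.
# The six bones then pass through half-alpha AnimDynamics with angular springs, gravity and limits, layering secondary physics over the procedural motion.
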AnimBlueprint /BRCosmetics/Characters/Player/Male/Medium/Heads/M_MED_SquibblySquid_Head/Meshes/M_MED_SquibblySquid_Head_Post_AnimBP
target_skeleton=&/Fortnite_Base_Head_SquibblySquid_Skeleton
graphs [
 AnimGraph {nodes [
   n1 = Get TentacleRotA() -> (TentacleRotA:Rotator)
   n2 = "Transform (Modify) Bone - Bone: dyn_hair_04"(ComponentPose=Local To Component(LocalPose=Input Pose()), Alpha=1, Translation="0,0,0", Rotation=@n1, Scale="1,1,1") {rotationMode=BMM_Additive, rotationSpace=BCS_BoneSpace, boneToModify {boneName=dyn_hair_04}}
   n3 = Get TentacleRotB() -> (TentacleRotB:Rotator)
   n4 = "Transform (Modify) Bone - Bone: dyn_hair_05_a"(
    ComponentPose="Transform (Modify) Bone - Bone: dyn_hair_05"(
     ComponentPose="Transform (Modify) Bone - Bone: dyn_hair_04_a"(ComponentPose=@n2, Alpha=1, Translation="0,0,0", Rotation=@n1, Scale="1,1,1") {rotationMode=BMM_Additive, rotationSpace=BCS_BoneSpace, boneToModify {boneName=dyn_hair_04_a}}
     Alpha=1
     Translation="0,0,0"
     Rotation=@n3
     Scale="1,1,1"
    ) {rotationMode=BMM_Additive, rotationSpace=BCS_BoneSpace, boneToModify {boneName=dyn_hair_05}}
    Alpha=1
    Translation="0,0,0"
    Rotation=@n3
    Scale="1,1,1"
   ) {rotationMode=BMM_Additive, rotationSpace=BCS_BoneSpace, boneToModify {boneName=dyn_hair_05_a}}
   n5 = Get TentacleRotC() -> (TentacleRotC:Rotator)
   n6 = Anim Dynamics - Bone: dyn_hair_04(
    ComponentPose="Transform (Modify) Bone - Bone: dyn_hair_06_a"(
     ComponentPose="Transform (Modify) Bone - Bone: dyn_hair_06"(ComponentPose=@n4, Alpha=1, Translation="0,0,0", Rotation=@n5, Scale="1,1,1") {rotationMode=BMM_Additive, rotationSpace=BCS_BoneSpace, boneToModify {boneName=dyn_hair_06}}
     Alpha=1
     Translation="0,0,0"
     Rotation=@n5
     Scale="1,1,1"
    ) {rotationMode=BMM_Additive, rotationSpace=BCS_BoneSpace, boneToModify {boneName=dyn_hair_06_a}}
    Alpha=0.5
    ExternalForce="0,0,0"
   ) {
    angularDampingOverride=1
    angularSpringConstant=2000
    bAngularSpring=true
    bOverrideAngularDamping=true
    bUseGravityOverride=true
    simulationSpace=RootRelative
    alphaBoolBlend {blendOption=Linear}
    boundBone {boneName=dyn_hair_04}
    componentLinearAccScale=<0.4 0.4 0.4>
    componentLinearVelScale=<0.4 0.4 0.4>
    gravityOverride {z=-980}
    physicsBodyDefinitions [{
      boundBone {boneName=dyn_hair_04}
      boxExtents=<10 10 10>
      constraintSetup {angularLimitsMax=<1 25 20>, angularLimitsMin=<1 -10 -20>, angularTarget=<1 0 0>}
      localJointOffset=<-1 0 0>
     }]
   }
   n7 = Anim Dynamics - Bone: dyn_hair_04_a(ComponentPose=@n6, Alpha=0.5, ExternalForce="0,0,0") {
    angularDampingOverride=1
    angularSpringConstant=2000
    bAngularSpring=true
    bOverrideAngularDamping=true
    bUseGravityOverride=true
    simulationSpace=RootRelative
    alphaBoolBlend {blendOption=Linear}
    boundBone {boneName=dyn_hair_04_a}
    componentLinearAccScale=<0.4 0.4 0.4>
    componentLinearVelScale=<0.4 0.4 0.4>
    gravityOverride {z=-980}
    physicsBodyDefinitions [{
      boundBone {boneName=dyn_hair_04_a}
      boxExtents=<10 10 10>
      constraintSetup {angularLimitsMax=<1 25 20>, angularLimitsMin=<1 -10 -20>, angularTarget=<1 0 0>}
      localJointOffset=<-1 0 0>
     }]
   }
   n8 = Anim Dynamics - Bone: dyn_hair_05(ComponentPose=@n7, Alpha=0.5, ExternalForce="0,0,0") {
    angularDampingOverride=1
    angularSpringConstant=2000
    bAngularSpring=true
    bOverrideAngularDamping=true
    bUseGravityOverride=true
    simulationSpace=RootRelative
    alphaBoolBlend {blendOption=Linear}
    boundBone {boneName=dyn_hair_05}
    componentLinearAccScale=<0.4 0.4 0.4>
    componentLinearVelScale=<0.4 0.4 0.4>
    gravityOverride {z=-980}
    physicsBodyDefinitions [{
      boundBone {boneName=dyn_hair_05}
      boxExtents=<10 10 10>
      constraintSetup {angularLimitsMax=<1 25 20>, angularLimitsMin=<1 -10 -20>, angularTarget=<1 0 0>}
      localJointOffset=<1 0 0>
     }]
   }
   n9 = Anim Dynamics - Bone: dyn_hair_05_a(ComponentPose=@n8, Alpha=0.5, ExternalForce="0,0,0") {
    angularDampingOverride=1
    angularSpringConstant=2000
    bAngularSpring=true
    bOverrideAngularDamping=true
    bUseGravityOverride=true
    simulationSpace=RootRelative
    alphaBoolBlend {blendOption=Linear}
    boundBone {boneName=dyn_hair_05_a}
    componentLinearAccScale=<0.4 0.4 0.4>
    componentLinearVelScale=<0.4 0.4 0.4>
    gravityOverride {z=-980}
    physicsBodyDefinitions [{
      boundBone {boneName=dyn_hair_05_a}
      boxExtents=<10 10 10>
      constraintSetup {angularLimitsMax=<1 25 20>, angularLimitsMin=<1 -10 -20>, angularTarget=<1 0 0>}
      localJointOffset=<1 0 0>
     }]
   }
   n10 = Anim Dynamics - Bone: dyn_hair_06(ComponentPose=@n9, Alpha=0.5, ExternalForce="0,0,0") {
    angularDampingOverride=1
    angularSpringConstant=2000
    bAngularSpring=true
    bOverrideAngularDamping=true
    bUseGravityOverride=true
    simulationSpace=RootRelative
    alphaBoolBlend {blendOption=Linear}
    boundBone {boneName=dyn_hair_06}
    componentLinearAccScale=<0.4 0.4 0.4>
    componentLinearVelScale=<0.4 0.4 0.4>
    gravityOverride {z=-980}
    physicsBodyDefinitions [{
      boundBone {boneName=dyn_hair_06}
      boxExtents=<10 10 10>
      constraintSetup {angularLimitsMax=<1 25 20>, angularLimitsMin=<1 -10 -20>, angularTarget=<1 0 0>}
      localJointOffset=<-1 0 0>
     }]
   }
   n11 = Anim Dynamics - Bone: dyn_hair_06_a(ComponentPose=@n10, Alpha=0.5, ExternalForce="0,0,0") {
    angularDampingOverride=1
    angularSpringConstant=2000
    bAngularSpring=true
    bOverrideAngularDamping=true
    bUseGravityOverride=true
    simulationSpace=RootRelative
    alphaBoolBlend {blendOption=Linear}
    boundBone {boneName=dyn_hair_06_a}
    componentLinearAccScale=<0.4 0.4 0.4>
    componentLinearVelScale=<0.4 0.4 0.4>
    gravityOverride {z=-980}
    physicsBodyDefinitions [{
      boundBone {boneName=dyn_hair_06_a}
      boxExtents=<10 10 10>
      constraintSetup {angularLimitsMax=<1 25 20>, angularLimitsMin=<1 -10 -20>, angularTarget=<1 0 0>}
      localJointOffset=<-1 0 0>
     }]
   }
   Output Pose(Result=Component To Local(ComponentPose=@n11))
  ]}
 EventGraph {nodes [
   Event BlueprintUpdateAnimation()
   n1 = float * float(A=GetGameTimeInSeconds(), B=5.5) {is=title K2Node_PromotableOperator}
   n2 = "Sin (Radians)"(A=float * float(A=1.25, B=@n1) {is=title K2Node_CommutativeAssociativeBinaryOperator})
   n3 = "Sin (Radians)"(A=float * float(A=2, B=@n1))
   n4 = "Sin (Radians)"(A=float * float(A=0.6, B=@n1))
   > Set TentacleRotA(TentacleRotA:Rotator=MakeRotator(Roll=float * float(A=1, B=@n4), Pitch=float * float(A=1, B=@n3), Yaw=float * float(A=1, B=@n2))) -> (Output_Get:Rotator)
   n5 = float * float(A=GetGameTimeInSeconds(), B=3) {is=title K2Node_PromotableOperator}
   n6 = "Sin (Radians)"(A=float * float(A=2, B=@n5) {is=title K2Node_CommutativeAssociativeBinaryOperator})
   n7 = "Sin (Radians)"(A=float * float(A=0.6, B=@n5))
   n8 = "Sin (Radians)"(A=float * float(A=1.25, B=@n5))
   > Set TentacleRotB(TentacleRotB:Rotator=MakeRotator(Roll=float * float(A=1.5, B=@n7), Pitch=float * float(A=1.5, B=@n6), Yaw=float * float(A=1.5, B=@n8))) -> (Output_Get:Rotator)
   n9 = float * float(A=GetGameTimeInSeconds(), B=4.5) {is=title K2Node_PromotableOperator}
   n10 = "Sin (Radians)"(A=float * float(A=1.25, B=@n9) {is=title K2Node_CommutativeAssociativeBinaryOperator})
   n11 = "Sin (Radians)"(A=float * float(A=2, B=@n9))
   n12 = "Sin (Radians)"(A=float * float(A=0.6, B=@n9))
   > Set TentacleRotC(TentacleRotC:Rotator=MakeRotator(Roll=float * float(A=2, B=@n12), Pitch=float * float(A=2, B=@n11), Yaw=float * float(A=2, B=@n10))) -> (Output_Get:Rotator)
   TryGetPawnOwner()
  ]}
]
variables [
 TentacleRotA {type=Rotator}
 TentacleRotB {type=Rotator}
 TentacleRotC {type=Rotator}
]
